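# Animation Blueprint for the female Dark Viking body cosmetic, adding secondary motion to belts, sword, skirt, tassels and rings.
# It copies the attached parent's pose, then runs one serial component-space chain of Trail controllers, Anim Dynamics and weighted multi-bone constraints.
# Constraints attach the skull, rings and front belt to skirt and body bones; several others have zero alpha.
# Skull dynamics and Flap nodes are unconnected, and the event graph has no logic.
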AnimBlueprint /BRCosmetics/Characters/Player/Female/Medium/Bodies/F_MED_Dark_Viking_01/Meshes/F_MED_Commando_DarkViking_AnimBP
target_skeleton=&/F_MED_Commando_DarkViking_Skeleton
graphs [
 AnimGraph {nodes [
   Anim Dynamics - Bone: dyn_skull(Alpha=1, ExternalForce="0,0,0") {
    angularDampingOverride=0.98
    angularSpringConstant=1500
    bAngularSpring=true
    bGravityOverrideInSimSpace=true
    bOverrideAngularDamping=true
    bUseSphericalLimits=true
    lODThreshold=1
    numSolverIterationsPostUpdate=4
    numSolverIterationsPreUpdate=16
    alphaBoolBlend {blendOption=Linear}
    boundBone {boneName=dyn_skull}
    componentLinearAccScale=<1 1 1>
    componentLinearVelScale=<1 1 1>
    physicsBodyDefinitions [{
      sphereCollisionRadius=0
      boundBone {boneName=dyn_skull}
      boxExtents=<10 10 20>
      constraintSetup {
       angularTargetAxis=AxisY
       angularLimitsMax=<60 60 60>
       angularLimitsMin=<-60 -60 -60>
       angularTarget=<0 0 0>
       linearAxesMax=<0 0 0.5>
       linearAxesMin=<0 0 -0.5>
      }
      localJointOffset=<0 0 -1>
     }]
    sphericalLimits [{limitRadius=40, drivingBone {boneName=thigh_twist_01_l}, sphereLocalOffset=<0 0 15.75>}]
   }
   n1 = Trail controller - Bone: dyn_belt_r_b!(ComponentPose=Local To Component(LocalPose=Copy Pose From Mesh() {bUseAttachedParent=true}), Alpha=0.65, FakeVelocity:Vector="0,0,0") -> (Pose) {
    chainBoneAxis=Z
    relaxationSpeedScale=8.5
    relaxationSpeedScaleInputProcessor {bInterpResult=true}
    rotationLimits [{limitMax=<180 180 180>}, {limitMax=<180 180 180>}]
    rotationOffsets [<0 0 0>, <0 0 0>]
    trailBone {boneName=dyn_belt_r_b}
   }
   n2 = Trail controller - Bone: dyn_belt_l_b!(ComponentPose=@n1, Alpha=0.45, FakeVelocity:Vector="0,0,0") -> (Pose) {
    chainBoneAxis=Z
    relaxationSpeedScale=8.5
    relaxationSpeedScaleInputProcessor {bInterpResult=true}
    rotationLimits [{limitMax=<180 180 180>}, {limitMax=<180 180 180>}]
    rotationOffsets [<0 0 0>, <0 0 0>]
    trailBone {boneName=dyn_belt_l_b}
   }
   n3 = Anim Dynamics - Bone: dyn_sword(ComponentPose=@n2, Alpha=0.345, ExternalForce="0,0,0") {
    angularDampingOverride=0.99
    angularSpringConstant=1500
    bAngularSpring=true
    bGravityOverrideInSimSpace=true
    bOverrideAngularDamping=true
    lODThreshold=1
    numSolverIterationsPostUpdate=4
    numSolverIterationsPreUpdate=16
    alphaBoolBlend {blendOption=Linear}
    boundBone {boneName=dyn_sword}
    physicsBodyDefinitions [{
      sphereCollisionRadius=0
      boundBone {boneName=dyn_sword}
      boxExtents=<10 10 20>
      constraintSetup {
       angularTargetAxis=AxisY
       twistAxis=AxisZ
       angularLimitsMax=<60 0 0>
       angularLimitsMin=<-60 0 0>
       angularTarget=<0 0 0>
       linearAxesMax=<0 0 0.5>
       linearAxesMin=<0 0 -0.5>
      }
      localJointOffset=<0 0 -2>
     }]
    sphericalLimits [{limitRadius=16, drivingBone {boneName=pelvis}, sphereLocalOffset=<-1 1 0>}]
   }
   n4 = "Constraint - dyn_skull (PP)"!(ComponentPose=@n3, Alpha=1, ConstraintWeights_0=1, ConstraintWeights_1:float=0) -> (Pose) {
    alphaBoolBlend {blendOption=Linear}
    boneToModify {boneName=dyn_skull}
    constraintSetup [{targetBone {boneName=dyn_skirt_out_02_l}}, {targetBone {boneName=pelvis}}]
    constraintWeights [1,1]
   }
   n5 = "Constraint - dyn_front_top_ring (PPPPP)"!(
    ComponentPose=@n4
    Alpha=1
    ConstraintWeights_0=0.0008
    ConstraintWeights_1:float=0.0156
    ConstraintWeights_2:float=0.0256
    ConstraintWeights_3:float=0.0025
    ConstraintWeights_4:float=0.9557
   ) -> (Pose) {
    alphaBoolBlend {blendOption=Linear}
    boneToModify {boneName=dyn_front_top_ring}
    constraintSetup [
     {targetBone {boneName=dyn_skirt_fr_01}}
     {targetBone {boneName=spine_01}}
     {targetBone {boneName=spine_02}}
     {targetBone {boneName=spine_04}}
     {targetBone {boneName=pelvis}}
    ]
    constraintWeights [1,1,1,1,1]
   }
   n6 = "Constraint - dyn_front_bottom_ring (PPPPPPPPPP)"!(
    ComponentPose=@n5
    Alpha=1
    ConstraintWeights_0=0.001
    ConstraintWeights_1:float=0.0005
    ConstraintWeights_2:float=0.026
    ConstraintWeights_3:float=0
    ConstraintWeights_4:float=0.0029
    ConstraintWeights_5:float=0.0048
    ConstraintWeights_6:float=0.0001
    ConstraintWeights_7:float=0.006
    ConstraintWeights_8:float=0.9461
    ConstraintWeights_9:float=0.0124
   ) -> (Pose) {
    alphaBoolBlend {blendOption=Linear}
    boneToModify {boneName=dyn_front_bottom_ring}
    constraintSetup [
     {targetBone {boneName=dyn_skirt_fr_01_r}}
     {targetBone {boneName=thigh_r}}
     {targetBone {boneName=dyn_skirt_fr_01}}
     {targetBone {boneName=dyn_skirt_fr_02}}
     {targetBone {boneName=spine_01}}
     {targetBone {boneName=spine_02}}
     {targetBone {boneName=spine_04}}
     {targetBone {boneName=thigh_l}}
     {targetBone {boneName=pelvis}}
     {targetBone {boneName=dyn_skirt_fr_01_l}}
    ]
    constraintWeights [1,1,1,1,1,1,1,1,1,1]
   }
   n7 = Anim Dynamics - Bone: dyn_skirt_out_01_l!(ComponentPose=@n6, ExternalForce="0,0,0") -> (Pose) {
    angularDampingOverride=0.99
    angularSpringConstant=15000
    bAngularSpring=true
    bGravityOverrideInSimSpace=true
    bOverrideAngularDamping=true
    numSolverIterationsPostUpdate=4
    numSolverIterationsPreUpdate=16
    alphaBoolBlend {blendOption=Linear}
    boundBone {boneName=dyn_skirt_out_01_l}
    physicsBodyDefinitions [{
      sphereCollisionRadius=0
      boundBone {boneName=dyn_skirt_out_01_l}
      boxExtents=<50 50 30>
      constraintSetup {angularTargetAxis=AxisZ, twistAxis=AxisZ, angularLimitsMax=<5 0 5>, angularLimitsMin=<-20 -15 -5>, angularTarget=<0 0 1>}
      localJointOffset=<0 0 -5>
     }]
   }
   n8 = Anim Dynamics - Bone: dyn_skirt_out_02_l!(ComponentPose=@n7, ExternalForce="0,0,0") -> (Pose) {
    angularDampingOverride=0.99
    angularSpringConstant=15000
    bAngularSpring=true
    bGravityOverrideInSimSpace=true
    bOverrideAngularDamping=true
    numSolverIterationsPostUpdate=4
    numSolverIterationsPreUpdate=16
    alphaBoolBlend {blendOption=Linear}
    boundBone {boneName=dyn_skirt_out_02_l}
    physicsBodyDefinitions [{
      sphereCollisionRadius=0
      boundBone {boneName=dyn_skirt_out_02_l}
      boxExtents=<50 50 30>
      constraintSetup {angularTargetAxis=AxisZ, twistAxis=AxisZ, angularLimitsMax=<20 0 5>, angularLimitsMin=<-20 -15 -5>, angularTarget=<0 0 1>}
      localJointOffset=<0 0 -5>
     }]
   }
   n9 = Anim Dynamics - Bone: dyn_skirt_bk_01_l!(ComponentPose=@n8, ExternalForce="0,0,0") -> (Pose) {
    angularDampingOverride=0.99
    angularSpringConstant=15000
    bAngularSpring=true
    bGravityOverrideInSimSpace=true
    bOverrideAngularDamping=true
    numSolverIterationsPostUpdate=4
    numSolverIterationsPreUpdate=16
    alphaBoolBlend {blendOption=Linear}
    boundBone {boneName=dyn_skirt_bk_01_l}
    physicsBodyDefinitions [{
      sphereCollisionRadius=0
      boundBone {boneName=dyn_skirt_bk_01_l}
      boxExtents=<50 50 30>
      constraintSetup {angularTargetAxis=AxisZ, twistAxis=AxisZ, angularLimitsMax=<20 0 5>, angularLimitsMin=<-20 -90 -5>, angularTarget=<0 0 1>}
      localJointOffset=<0 0 -5>
     }]
   }
   n10 = Anim Dynamics - Bone: dyn_skirt_bk_02_l!(ComponentPose=@n9, ExternalForce="0,0,0") -> (Pose) {
    angularDampingOverride=0.99
    angularSpringConstant=15000
    bAngularSpring=true
    bGravityOverrideInSimSpace=true
    bOverrideAngularDamping=true
    numSolverIterationsPostUpdate=4
    numSolverIterationsPreUpdate=16
    alphaBoolBlend {blendOption=Linear}
    boundBone {boneName=dyn_skirt_bk_02_l}
    physicsBodyDefinitions [{
      sphereCollisionRadius=0
      boundBone {boneName=dyn_skirt_bk_02_l}
      boxExtents=<50 50 30>
      constraintSetup {angularTargetAxis=AxisZ, twistAxis=AxisZ, angularLimitsMax=<20 0 5>, angularLimitsMin=<-20 -90 -5>, angularTarget=<0 0 1>}
      localJointOffset=<0 0 -5>
     }]
   }
   n11 = Anim Dynamics - Bone: dyn_skirt_frdiag_01_l!(ComponentPose=@n10, ExternalForce="0,0,0") -> (Pose) {
    angularDampingOverride=0.99
    angularSpringConstant=15000
    bAngularSpring=true
    bGravityOverrideInSimSpace=true
    bOverrideAngularDamping=true
    numSolverIterationsPostUpdate=4
    numSolverIterationsPreUpdate=16
    alphaBoolBlend {blendOption=Linear}
    boundBone {boneName=dyn_skirt_frdiag_01_l}
    physicsBodyDefinitions [{
      sphereCollisionRadius=0
      boundBone {boneName=dyn_skirt_frdiag_01_l}
      boxExtents=<50 50 30>
      constraintSetup {angularTargetAxis=AxisZ, twistAxis=AxisZ, angularLimitsMax=<10 0 0>, angularLimitsMin=<-10 -15 0>, angularTarget=<0 0 1>}
      localJointOffset=<0 0 -5>
     }]
   }
   n12 = Anim Dynamics - Bone: dyn_skirt_frdiag_02_l!(ComponentPose=@n11, ExternalForce="0,0,0") -> (Pose) {
    angularDampingOverride=0.99
    angularSpringConstant=15000
    bAngularSpring=true
    bGravityOverrideInSimSpace=true
    bOverrideAngularDamping=true
    numSolverIterationsPostUpdate=4
    numSolverIterationsPreUpdate=16
    alphaBoolBlend {blendOption=Linear}
    boundBone {boneName=dyn_skirt_frdiag_02_l}
    physicsBodyDefinitions [{
      sphereCollisionRadius=0
      boundBone {boneName=dyn_skirt_frdiag_02_l}
      boxExtents=<50 50 30>
      constraintSetup {angularTargetAxis=AxisZ, twistAxis=AxisZ, angularLimitsMax=<10 0 5>, angularLimitsMin=<-10 -15 -5>, angularTarget=<0 0 1>}
      localJointOffset=<0 0 -5>
     }]
   }
   n13 = Anim Dynamics - Bone: dyn_skirt_out_01_r!(ComponentPose=@n12, ExternalForce="0,0,0") -> (Pose) {
    angularDampingOverride=0.99
    angularSpringConstant=15000
    bAngularSpring=true
    bGravityOverrideInSimSpace=true
    bOverrideAngularDamping=true
    numSolverIterationsPostUpdate=4
    numSolverIterationsPreUpdate=16
    alphaBoolBlend {blendOption=Linear}
    boundBone {boneName=dyn_skirt_out_01_r}
    physicsBodyDefinitions [{
      sphereCollisionRadius=0
      boundBone {boneName=dyn_skirt_out_01_r}
      boxExtents=<50 50 30>
      constraintSetup {angularTargetAxis=AxisZ, twistAxis=AxisZ, angularLimitsMax=<5 0 5>, angularLimitsMin=<-20 -15 -5>, angularTarget=<0 0 1>}
      localJointOffset=<0 0 5>
     }]
   }
   n14 = Anim Dynamics - Bone: dyn_skirt_out_02_r!(ComponentPose=@n13, ExternalForce="0,0,0") -> (Pose) {
    angularDampingOverride=0.99
    angularSpringConstant=15000
    bAngularSpring=true
    bGravityOverrideInSimSpace=true
    bOverrideAngularDamping=true
    numSolverIterationsPostUpdate=4
    numSolverIterationsPreUpdate=16
    alphaBoolBlend {blendOption=Linear}
    boundBone {boneName=dyn_skirt_out_02_r}
    physicsBodyDefinitions [{
      sphereCollisionRadius=0
      boundBone {boneName=dyn_skirt_out_02_r}
      boxExtents=<50 50 30>
      constraintSetup {angularTargetAxis=AxisZ, twistAxis=AxisZ, angularLimitsMax=<20 0 5>, angularLimitsMin=<-20 -15 -5>, angularTarget=<0 0 1>}
      localJointOffset=<0 0 5>
     }]
   }
   n15 = Anim Dynamics - Bone: dyn_skirt_fr_01_r!(ComponentPose=@n14, ExternalForce="0,0,0") -> (Pose) {
    angularDampingOverride=0.99
    angularSpringConstant=15000
    bAngularSpring=true
    bGravityOverrideInSimSpace=true
    bOverrideAngularDamping=true
    numSolverIterationsPostUpdate=4
    numSolverIterationsPreUpdate=16
    alphaBoolBlend {blendOption=Linear}
    boundBone {boneName=dyn_skirt_fr_01_r}
    physicsBodyDefinitions [{
      sphereCollisionRadius=0
      boundBone {boneName=dyn_skirt_fr_01_r}
      boxExtents=<50 50 30>
      constraintSetup {angularTargetAxis=AxisZ, twistAxis=AxisZ, angularLimitsMax=<0 0 0>, angularLimitsMin=<0 -15 0>, angularTarget=<0 0 1>}
      localJointOffset=<0 0 5>
     }]
   }
   n16 = Anim Dynamics - Bone: dyn_skirt_fr_02_r!(ComponentPose=@n15, ExternalForce="0,0,0") -> (Pose) {
    angularDampingOverride=0.99
    angularSpringConstant=15000
    bAngularSpring=true
    bGravityOverrideInSimSpace=true
    bOverrideAngularDamping=true
    numSolverIterationsPostUpdate=4
    numSolverIterationsPreUpdate=16
    alphaBoolBlend {blendOption=Linear}
    boundBone {boneName=dyn_skirt_fr_02_r}
    physicsBodyDefinitions [{
      sphereCollisionRadius=0
      boundBone {boneName=dyn_skirt_fr_02_r}
      boxExtents=<50 50 30>
      constraintSetup {angularTargetAxis=AxisZ, twistAxis=AxisZ, angularLimitsMax=<0 0 5>, angularLimitsMin=<-5 -15 -5>, angularTarget=<0 0 1>}
      localJointOffset=<0 0 5>
     }]
   }
   n17 = Anim Dynamics - Bone: dyn_skirt_bk_01_r!(ComponentPose=@n16, ExternalForce="0,0,0") -> (Pose) {
    angularDampingOverride=0.99
    angularSpringConstant=15000
    bAngularSpring=true
    bGravityOverrideInSimSpace=true
    bOverrideAngularDamping=true
    numSolverIterationsPostUpdate=4
    numSolverIterationsPreUpdate=16
    alphaBoolBlend {blendOption=Linear}
    boundBone {boneName=dyn_skirt_bk_01_r}
    physicsBodyDefinitions [{
      sphereCollisionRadius=0
      boundBone {boneName=dyn_skirt_bk_01_r}
      boxExtents=<50 50 30>
      constraintSetup {angularTargetAxis=AxisZ, twistAxis=AxisZ, angularLimitsMax=<20 0 5>, angularLimitsMin=<-20 -90 -5>, angularTarget=<0 0 1>}
      localJointOffset=<0 0 5>
     }]
   }
   n18 = Anim Dynamics - Bone: dyn_skirt_bk_02_r!(ComponentPose=@n17, ExternalForce="0,0,0") -> (Pose) {
    angularDampingOverride=0.99
    angularSpringConstant=15000
    bAngularSpring=true
    bGravityOverrideInSimSpace=true
    bOverrideAngularDamping=true
    numSolverIterationsPostUpdate=4
    numSolverIterationsPreUpdate=16
    alphaBoolBlend {blendOption=Linear}
    boundBone {boneName=dyn_skirt_bk_02_r}
    physicsBodyDefinitions [{
      sphereCollisionRadius=0
      boundBone {boneName=dyn_skirt_bk_02_r}
      boxExtents=<50 50 30>
      constraintSetup {angularTargetAxis=AxisZ, twistAxis=AxisZ, angularLimitsMax=<20 0 5>, angularLimitsMin=<-20 -90 -5>, angularTarget=<0 0 1>}
      localJointOffset=<0 0 5>
     }]
   }
   n19 = Anim Dynamics - Bone: dyn_skirt_frdiag_01_r!(ComponentPose=@n18, ExternalForce="0,0,0") -> (Pose) {
    angularDampingOverride=0.99
    angularSpringConstant=15000
    bAngularSpring=true
    bGravityOverrideInSimSpace=true
    bOverrideAngularDamping=true
    numSolverIterationsPostUpdate=4
    numSolverIterationsPreUpdate=16
    alphaBoolBlend {blendOption=Linear}
    boundBone {boneName=dyn_skirt_frdiag_01_r}
    physicsBodyDefinitions [{
      sphereCollisionRadius=0
      boundBone {boneName=dyn_skirt_frdiag_01_r}
      boxExtents=<50 50 30>
      constraintSetup {angularTargetAxis=AxisZ, twistAxis=AxisZ, angularLimitsMax=<10 0 0>, angularLimitsMin=<-10 -15 0>, angularTarget=<0 0 1>}
      localJointOffset=<0 0 5>
     }]
   }
   n20 = Anim Dynamics - Bone: dyn_skirt_frdiag_02_r!(ComponentPose=@n19, ExternalForce="0,0,0") -> (Pose) {
    angularDampingOverride=0.99
    angularSpringConstant=15000
    bAngularSpring=true
    bGravityOverrideInSimSpace=true
    bOverrideAngularDamping=true
    numSolverIterationsPostUpdate=4
    numSolverIterationsPreUpdate=16
    alphaBoolBlend {blendOption=Linear}
    boundBone {boneName=dyn_skirt_frdiag_02_r}
    physicsBodyDefinitions [{
      sphereCollisionRadius=0
      boundBone {boneName=dyn_skirt_frdiag_02_r}
      boxExtents=<50 50 30>
      constraintSetup {angularTargetAxis=AxisZ, twistAxis=AxisZ, angularLimitsMax=<10 0 5>, angularLimitsMin=<-10 -15 -5>, angularTarget=<0 0 1>}
      localJointOffset=<0 0 5>
     }]
   }
   n21 = "Constraint - dyn_skirt_fr_01 (TTRR)"!(ComponentPose=@n20, Alpha=0, ConstraintWeights_0=1, ConstraintWeights_1:float=1, ConstraintWeights_2:float=1, ConstraintWeights_3:float=1) -> (Pose) {
    alphaBoolBlend {blendOption=Linear}
    boneToModify {boneName=dyn_skirt_fr_01}
    constraintSetup [
     {transformType=Translation, targetBone {boneName=dyn_skirt_fr_01_l}}
     {transformType=Translation, targetBone {boneName=dyn_skirt_fr_01_r}}
     {transformType=Rotation, targetBone {boneName=dyn_skirt_fr_01_l}}
     {transformType=Rotation, targetBone {boneName=dyn_skirt_fr_01_r}}
    ]
    constraintWeights [1,1,1,1]
   }
   n22 = "Constraint - dyn_skirt_fr_02 (TTRR)"!(ComponentPose=@n21, Alpha=0, ConstraintWeights_0=1, ConstraintWeights_1:float=1, ConstraintWeights_2:float=1, ConstraintWeights_3:float=1) -> (Pose) {
    alphaBoolBlend {blendOption=Linear}
    boneToModify {boneName=dyn_skirt_fr_02}
    constraintSetup [
     {transformType=Translation, targetBone {boneName=dyn_skirt_fr_02_l}}
     {transformType=Translation, targetBone {boneName=dyn_skirt_fr_02_r}}
     {transformType=Rotation, targetBone {boneName=dyn_skirt_fr_02_l}}
     {transformType=Rotation, targetBone {boneName=dyn_skirt_fr_02_r}}
    ]
    constraintWeights [1,1,1,1]
   }
   n23 = "Constraint - dyn_front_belt_01 (PPPPPPPPPPPP)"!(
    ComponentPose=@n22
    Alpha=1
    ConstraintWeights_0=0.0034
    ConstraintWeights_1:float=0.1296
    ConstraintWeights_2:float=0.0018
    ConstraintWeights_3:float=0.227
    ConstraintWeights_4:float=0.2079
    ConstraintWeights_5:float=0.0994
    ConstraintWeights_6:float=0.0531
    ConstraintWeights_7:float=0.0629
    ConstraintWeights_8:float=0.2071
    ConstraintWeights_9:float=0.0014
    ConstraintWeights_10:float=0.0028
    ConstraintWeights_11:float=0.0036
   ) -> (Pose) {
    alphaBoolBlend {blendOption=Linear}
    boneToModify {boneName=dyn_front_belt_01}
    constraintSetup [
     {targetBone {boneName=dyn_skirt_bk_02_r}}
     {targetBone {boneName=dyn_skirt_fr_01_r}}
     {targetBone {boneName=dyn_skirt_out_01_l}}
     {targetBone {boneName=dyn_skirt_fr_01}}
     {targetBone {boneName=dyn_skirt_fr_02}}
     {targetBone {boneName=dyn_skirt_fr_02_l}}
     {targetBone {boneName=dyn_skirt_fr_02_r}}
     {targetBone {boneName=pelvis}}
     {targetBone {boneName=dyn_skirt_fr_01_l}}
     {targetBone {boneName=dyn_skirt_out_01_r}}
     {targetBone {boneName=dyn_skirt_out_02_r}}
     {targetBone {boneName=dyn_skirt_out_02_l}}
    ]
    constraintWeights [1,1,1,1,1,1,1,1,1,1,1,1]
   }
   n24 = Anim Dynamics - Bone: dyn_front_tassle_l(ComponentPose=@n23, Alpha=1, ExternalForce="0,0,0") {
    angularDampingOverride=0.99
    angularSpringConstant=15000
    bGravityOverrideInSimSpace=true
    bOverrideAngularDamping=true
    numSolverIterationsPostUpdate=4
    numSolverIterationsPreUpdate=16
    alphaBoolBlend {blendOption=Linear}
    boundBone {boneName=dyn_front_tassle_l}
    physicsBodyDefinitions [{
      sphereCollisionRadius=0
      boundBone {boneName=dyn_front_tassle_l}
      boxExtents=<50 50 30>
      constraintSetup {angularTargetAxis=AxisZ, twistAxis=AxisZ, angularLimitsMax=<5 0 5>, angularLimitsMin=<-20 -15 -5>, angularTarget=<0 0 1>}
      localJointOffset=<0 0 -5>
     }]
   }
   n25 = Anim Dynamics - Bone: dyn_front_tassle_r(ComponentPose=@n24, Alpha=1, ExternalForce="0,0,0") {
    angularDampingOverride=0.99
    angularSpringConstant=15000
    bGravityOverrideInSimSpace=true
    bOverrideAngularDamping=true
    numSolverIterationsPostUpdate=4
    numSolverIterationsPreUpdate=16
    alphaBoolBlend {blendOption=Linear}
    boundBone {boneName=dyn_front_tassle_r}
    physicsBodyDefinitions [{
      sphereCollisionRadius=0
      boundBone {boneName=dyn_front_tassle_r}
      boxExtents=<50 50 30>
      constraintSetup {angularTargetAxis=AxisZ, twistAxis=AxisZ, angularLimitsMax=<20 0 5>, angularLimitsMin=<-20 -15 -5>, angularTarget=<0 0 1>}
      localJointOffset=<0 0 -5>
     }]
   }
   n26 = Anim Dynamics - Bone: dyn_front_bottom_ring_a!(ComponentPose=@n25, ExternalForce="0,0,0") -> (Pose) {
    angularDampingOverride=0.99
    angularSpringConstant=15000
    bAngularSpring=true
    bGravityOverrideInSimSpace=true
    bOverrideAngularDamping=true
    numSolverIterationsPostUpdate=4
    numSolverIterationsPreUpdate=16
    alphaBoolBlend {blendOption=Linear}
    boundBone {boneName=dyn_front_bottom_ring_a}
    physicsBodyDefinitions [{
      sphereCollisionRadius=0
      boundBone {boneName=dyn_front_bottom_ring_a}
      boxExtents=<50 50 30>
      constraintSetup {angularTargetAxis=AxisZ, twistAxis=AxisZ, angularLimitsMax=<5 0 5>, angularLimitsMin=<-20 -15 -5>, angularTarget=<0 0 1>}
      localJointOffset=<0 0 -5>
     }]
   }
   n27 = Anim Dynamics - Bone: dyn_front_top_ring_a!(ComponentPose=@n26, ExternalForce="0,0,0") -> (Pose) {
    angularDampingOverride=0.99
    angularSpringConstant=15000
    bAngularSpring=true
    bGravityOverrideInSimSpace=true
    bOverrideAngularDamping=true
    numSolverIterationsPostUpdate=4
    numSolverIterationsPreUpdate=16
    alphaBoolBlend {blendOption=Linear}
    boundBone {boneName=dyn_front_top_ring_a}
    physicsBodyDefinitions [{
      sphereCollisionRadius=0
      boundBone {boneName=dyn_front_top_ring_a}
      boxExtents=<50 50 30>
      constraintSetup {angularTargetAxis=AxisZ, twistAxis=AxisZ, angularLimitsMax=<20 0 5>, angularLimitsMin=<-20 -15 -5>, angularTarget=<0 0 1>}
      localJointOffset=<0 0 -5>
     }]
   }
   n28 = "Anim Dynamics - Chain: dyn_skirt_fr_01 -> dyn_skirt_fr_02"(ComponentPose=@n27, Alpha=0.25, ExternalForce="0,0,0") {
    angularDampingOverride=0.98
    angularSpringConstant=500
    bChain=true
    bGravityOverrideInSimSpace=true
    bOverrideAngularDamping=true
    bOverrideLinearDamping=true
    lODThreshold=1
    linearDampingOverride=0.98
    numSolverIterationsPostUpdate=4
    numSolverIterationsPreUpdate=16
    alphaBoolBlend {blendOption=Linear}
    boundBone {boneName=dyn_skirt_fr_01}
    chainEnd {boneName=dyn_skirt_fr_02}
    componentLinearAccScale=<1 1 1>
    componentLinearVelScale=<1 1 1>
    physicsBodyDefinitions [{
      sphereCollisionRadius=0
      boundBone {boneName=dyn_skirt_fr_01}
      boxExtents=<10 10 20>
      constraintSetup {
       angularTargetAxis=AxisY
       twistAxis=AxisZ
       angularLimitsMax=<30 10 0>
       angularLimitsMin=<-30 -10 0>
       angularTarget=<0 0 0>
       linearAxesMax=<0 0 0.5>
       linearAxesMin=<0 0 -0.5>
      }
      localJointOffset=<0 0 -2>
     }]
    sphericalLimits [{limitRadius=16, drivingBone {boneName=pelvis}, sphereLocalOffset=<-1 1 0>}]
   }
   n29 = "Constraint - dyn_skull (PPPP)"!(
    ComponentPose=@n28
    Alpha=0
    ConstraintWeights_0=0.0206
    ConstraintWeights_1:float=0.0319
    ConstraintWeights_2:float=0.0269
    ConstraintWeights_3:float=0.9206
   ) -> (Pose) {
    alphaBoolBlend {blendOption=Linear}
    boneToModify {boneName=dyn_skull}
    constraintSetup [
     {targetBone {boneName=spine_01}}
     {targetBone {boneName=spine_02}}
     {targetBone {boneName=dyn_skirt_out_01_l}}
     {targetBone {boneName=pelvis}}
    ]
    constraintWeights [1,1,1,1]
   }
   Output Pose(Result=Component To Local(ComponentPose=@n29)) {layerGroup=None}
   Flap Node - Bone: dyn_skull!(ComponentPose:Component Space Pose Link Structure, Alpha:float=1) -> (Pose:Component Space Pose Link Structure) {
    is=title FortAnimGraphNode_Flap
    targetClampMax=3.14
    targetComponent=RotationY
    alphaBoolBlend {blendOption=Linear}
    sourceDrives [{
      bUseRange=true
      drivingCurve=null
      multiplier=-1
      rangeMax=1.7453
      rangeMin=-1.7453
      remappedMax=1.7453
      remappedMin=-1.7453
      sourceComponent=RotationX
      useQuaternion=true
      sourceBone {boneName=thigh_l}
     }]
    targetBone {boneName=dyn_skull}
   }
  ]}
 EventGraph {nodes [Event BlueprintUpdateAnimation()]}
]
variables []
