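# Physics asset that simulates the dynamic hair bones on the female medium Dark Viking Commando head.
# Capsule bodies cover the hair bones, head and upper spine (spine_05), whose much larger capsule represents the upper torso.
# Hair roots 01 and 02 rotate freely on the head, each carrying an angle-limited child segment.
# Hair 05–08 and 12 attach directly to the head with tight asymmetric swing limits, and the head rotates freely on spine_05.
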PhysicsAsset /BRCosmetics/Characters/Player/Female/Medium/Bodies/F_MED_Dark_Viking_01/Meshes/F_MED_Commando_DarkViking_Head_01_dyn_hair
bodies {
 dyn_DVK_hair_01 [{shapeType=Capsule, length=1, radius=1, shapeName=dyn_DVK_hair_01_capsule}]
 dyn_DVK_hair_01_a [{shapeType=Capsule, length=3.1474, radius=2.1064, shapeName=dyn_DVK_hair_01_a_capsule, center=<0 2.8 -3>, rotation=<rot 0 0 -30>}]
 dyn_DVK_hair_02 [{shapeType=Capsule, length=1, radius=1, shapeName=dyn_DVK_hair_02_capsule}]
 dyn_DVK_hair_02_a [{shapeType=Capsule, length=3.1474, radius=2.1064, shapeName=dyn_DVK_hair_01_a_capsule, center=<0 2.5 -3>, rotation=<rot 0 0 -30>}]
 dyn_DVK_hair_05 [{shapeType=Capsule, length=2.5231, radius=2.4946, shapeName=dyn_DVK_hair_08_capsule, center=<0 -3.5 -6>, rotation=<rot 0 0 29.9992>}]
 dyn_DVK_hair_06 [{shapeType=Capsule, length=2.5231, radius=2.4946, shapeName=dyn_DVK_hair_08_capsule, center=<0 -3 -6>, rotation=<rot 0 0 29.9992>}]
 dyn_DVK_hair_07 [{shapeType=Capsule, length=2.5231, radius=2.4946, shapeName=dyn_DVK_hair_08_capsule, center=<0 -3 -6>, rotation=<rot 0 0 29.9992>}]
 dyn_DVK_hair_08 [{shapeType=Capsule, length=2.5231, radius=2.4946, shapeName=dyn_DVK_hair_08_capsule, center=<0 -3.5 -6>, rotation=<rot 0 0 29.9992>}]
 dyn_DVK_hair_12 [{shapeType=Capsule, length=2, radius=2, shapeName=dyn_DVK_hair_08_capsule, center=<0 0 -2>}]
 head [{shapeType=Capsule, length=2, radius=2, shapeName=head_capsule}]
 spine_05 [{shapeType=Capsule, length=17.25, radius=10.5, center=<2.8079 -3.1815 0.0001>}]
}
constraints [
 {
  bone1Name=dyn_DVK_hair_01
  bone2Name=head
  swing1LimitDegrees=45
  swing1Motion=Free
  swing2LimitDegrees=45
  swing2Motion=Free
  twistLimitDegrees=45
  twistMotion=Free
 }
 {bone1Name=dyn_DVK_hair_01_a, bone2Name=dyn_DVK_hair_01, swing1LimitDegrees=2, swing2LimitDegrees=15, twistLimitDegrees=30}
 {
  bone1Name=dyn_DVK_hair_02
  bone2Name=head
  swing1LimitDegrees=45
  swing1Motion=Free
  swing2LimitDegrees=45
  swing2Motion=Free
  twistLimitDegrees=45
  twistMotion=Free
 }
 {bone1Name=dyn_DVK_hair_02_a, bone2Name=dyn_DVK_hair_02, swing1LimitDegrees=2, swing2LimitDegrees=15, twistLimitDegrees=30}
 {bone1Name=dyn_DVK_hair_05, bone2Name=head, swing1LimitDegrees=2, swing2LimitDegrees=15, twistLimitDegrees=45}
 {bone1Name=dyn_DVK_hair_06, bone2Name=head, swing1LimitDegrees=2, swing2LimitDegrees=15, twistLimitDegrees=45}
 {bone1Name=dyn_DVK_hair_07, bone2Name=head, swing1LimitDegrees=2, swing2LimitDegrees=15, twistLimitDegrees=45}
 {bone1Name=dyn_DVK_hair_08, bone2Name=head, swing1LimitDegrees=2, swing2LimitDegrees=15, twistLimitDegrees=45}
 {
  bone1Name=head
  bone2Name=spine_05
  swing1LimitDegrees=45
  swing1Motion=Free
  swing2LimitDegrees=45
  swing2Motion=Free
  twistLimitDegrees=45
  twistMotion=Free
 }
 {bone1Name=dyn_DVK_hair_12, bone2Name=head, swing1LimitDegrees=5, swing2LimitDegrees=20, twistLimitDegrees=20}
]
