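# Head AnimBP for the female Dark Viking commando cosmetic, adding secondary hair motion to a pose copied from the attached parent mesh.
# A light AnimDynamics pass on hair_13 is cached, then feeds two head-space RigidBody simulations, blended by bIsSkyDiving between skydiving and regular hair physics assets.
# Each update, a deprecated gravity-override function sets both simulations' gravity.
# Several other AnimDynamics hair chains remain in the graph, unconnected to the output.
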AnimBlueprint /BRCosmetics/Characters/Player/Female/Medium/Bodies/F_MED_Dark_Viking_01/Meshes/F_MED_Commando_DarkViking_Head_01_AnimBP
target_skeleton=&/F_MED_Commando_DarkViking_Head_01_Skeleton
graphs [
 AnimGraph {nodes [
   n1 = Anim Dynamics - Bone: dyn_DVK_hair_07(Alpha=1, ExternalForce="0,0,0") {
    angularBiasOverride=1
    angularDampingOverride=0.98
    angularSpringConstant=500
    bAngularSpring=true
    bGravityOverrideInSimSpace=true
    bOverrideAngularDamping=true
    bUsePlanarLimit=false
    bUseSphericalLimits=true
    gravityScale=0.5
    lODThreshold=1
    numSolverIterationsPostUpdate=4
    numSolverIterationsPreUpdate=16
    alphaBoolBlend {blendOption=Linear}
    boundBone {boneName=dyn_DVK_hair_07}
    componentLinearAccScale=<1 1 1>
    componentLinearVelScale=<1 1 1>
    physicsBodyDefinitions [{
      sphereCollisionRadius=0
      boundBone {boneName=dyn_DVK_hair_07}
      boxExtents=<20 35 2>
      constraintSetup {angularTargetAxis=AxisZ, twistAxis=AxisY, angularLimitsMax=<80 40 0>, angularLimitsMin=<-80 -40 0>, angularTarget=<0 0 0>}
      localJointOffset=<0 0 -10>
     }]
    sphericalLimits [
     {limitRadius=15, drivingBone {boneName=spine_05}, sphereLocalOffset=<10 -2 0>}
     {limitRadius=10, drivingBone {boneName=clavicle_l}, sphereLocalOffset=<10 -2 0>}
    ]
   }
   n2 = Anim Dynamics - Bone: dyn_DVK_hair_06(ComponentPose=@n1, Alpha=1, ExternalForce="0,0,0") {
    angularBiasOverride=1
    angularDampingOverride=0.98
    angularSpringConstant=500
    bAngularSpring=true
    bGravityOverrideInSimSpace=true
    bOverrideAngularDamping=true
    bUseSphericalLimits=true
    gravityScale=0.5
    lODThreshold=1
    numSolverIterationsPostUpdate=4
    numSolverIterationsPreUpdate=16
    alphaBoolBlend {blendOption=Linear}
    boundBone {boneName=dyn_DVK_hair_06}
    componentLinearAccScale=<1 1 1>
    componentLinearVelScale=<1 1 1>
    physicsBodyDefinitions [{
      sphereCollisionRadius=0
      boundBone {boneName=dyn_DVK_hair_06}
      boxExtents=<20 35 2>
      constraintSetup {angularTargetAxis=AxisZ, twistAxis=AxisY, angularLimitsMax=<80 40 0>, angularLimitsMin=<-80 -40 0>, angularTarget=<0 0 0>}
      localJointOffset=<0 0 -10>
     }]
    sphericalLimits [
     {limitRadius=15, drivingBone {boneName=spine_05}, sphereLocalOffset=<10 -2 0>}
     {limitRadius=20, drivingBone {boneName=clavicle_r}, sphereLocalOffset=<-10 -5 10>}
    ]
   }
   n3 = Anim Dynamics - Bone: dyn_DVK_hair_08(ComponentPose=@n2, Alpha=1, ExternalForce="0,0,0") {
    angularBiasOverride=1
    angularDampingOverride=0.98
    angularSpringConstant=500
    bAngularSpring=true
    bGravityOverrideInSimSpace=true
    bOverrideAngularDamping=true
    bUseSphericalLimits=true
    gravityScale=0.5
    lODThreshold=1
    numSolverIterationsPostUpdate=4
    numSolverIterationsPreUpdate=16
    alphaBoolBlend {blendOption=Linear}
    boundBone {boneName=dyn_DVK_hair_08}
    componentLinearAccScale=<1 1 1>
    componentLinearVelScale=<1 1 1>
    physicsBodyDefinitions [{
      sphereCollisionRadius=0
      boundBone {boneName=dyn_DVK_hair_08}
      boxExtents=<20 35 2>
      constraintSetup {angularTargetAxis=AxisZ, twistAxis=AxisY, angularLimitsMax=<80 40 0>, angularLimitsMin=<-80 -40 0>, angularTarget=<0 0 0>}
      localJointOffset=<0 0 -10>
     }]
    sphericalLimits [
     {limitRadius=15, drivingBone {boneName=spine_05}, sphereLocalOffset=<10 -2 0>}
     {limitRadius=10, drivingBone {boneName=clavicle_l}, sphereLocalOffset=<10 -2 0>}
    ]
   }
   n4 = Anim Dynamics - Bone: dyn_DVK_hair_05(ComponentPose=@n3, Alpha=1, ExternalForce="0,0,0") {
    angularBiasOverride=1
    angularDampingOverride=0.98
    angularSpringConstant=500
    bAngularSpring=true
    bGravityOverrideInSimSpace=true
    bOverrideAngularDamping=true
    bUseSphericalLimits=true
    gravityScale=0.5
    lODThreshold=1
    numSolverIterationsPostUpdate=4
    numSolverIterationsPreUpdate=16
    alphaBoolBlend {blendOption=Linear}
    boundBone {boneName=dyn_DVK_hair_05}
    componentLinearAccScale=<1 1 1>
    componentLinearVelScale=<1 1 1>
    physicsBodyDefinitions [{
      sphereCollisionRadius=0
      boundBone {boneName=dyn_DVK_hair_05}
      boxExtents=<20 35 2>
      constraintSetup {angularTargetAxis=AxisZ, twistAxis=AxisY, angularLimitsMax=<80 40 0>, angularLimitsMin=<-80 -40 0>, angularTarget=<0 0 0>}
      localJointOffset=<0 0 -10>
     }]
    sphericalLimits [
     {limitRadius=15, drivingBone {boneName=spine_05}, sphereLocalOffset=<10 -2 0>}
     {limitRadius=20, drivingBone {boneName=clavicle_r}, sphereLocalOffset=<-10 -5 10>}
    ]
   }
   n5 = Anim Dynamics - Bone: dyn_DVK_hair_02_a(ComponentPose=@n4, Alpha=0.9, ExternalForce="0,0,0") {
    angularBiasOverride=1
    angularDampingOverride=0.998
    angularSpringConstant=500
    bAngularSpring=true
    bGravityOverrideInSimSpace=true
    bOverrideAngularDamping=true
    bOverrideLinearDamping=true
    bUseSphericalLimits=true
    lODThreshold=1
    linearDampingOverride=0.998
    numSolverIterationsPostUpdate=4
    numSolverIterationsPreUpdate=16
    alphaBoolBlend {blendOption=Linear}
    boundBone {boneName=dyn_DVK_hair_02_a}
    componentLinearAccScale=<0.1 0.1 0.1>
    componentLinearVelScale=<0.1 0.1 0.1>
    physicsBodyDefinitions [{
      sphereCollisionRadius=0
      boundBone {boneName=dyn_DVK_hair_02_a}
      boxExtents=<5 5 2>
      constraintSetup {angularTargetAxis=AxisZ, twistAxis=AxisY, angularLimitsMax=<80 40 0>, angularLimitsMin=<-60 -40 0>, angularTarget=<0 0 0>}
      localJointOffset=<0 0 -10>
     }]
    sphericalLimits [{limitRadius=15, drivingBone {boneName=spine_05}, sphereLocalOffset=<2 -6 0>}]
   }
   Anim Dynamics - Bone: dyn_DVK_hair_01_a(ComponentPose=@n5, Alpha=0.9, ExternalForce="0,0,0") {
    angularBiasOverride=1
    angularDampingOverride=0.998
    angularSpringConstant=500
    bAngularSpring=true
    bGravityOverrideInSimSpace=true
    bOverrideAngularDamping=true
    bOverrideLinearDamping=true
    bUseSphericalLimits=true
    lODThreshold=1
    linearDampingOverride=0.998
    numSolverIterationsPostUpdate=4
    numSolverIterationsPreUpdate=16
    alphaBoolBlend {blendOption=Linear}
    boundBone {boneName=dyn_DVK_hair_01_a}
    componentLinearAccScale=<0.1 0.1 0.1>
    componentLinearVelScale=<0.1 0.1 0.1>
    physicsBodyDefinitions [{
      sphereCollisionRadius=0
      boundBone {boneName=dyn_DVK_hair_01_a}
      boxExtents=<5 5 2>
      constraintSetup {angularTargetAxis=AxisZ, twistAxis=AxisY, angularLimitsMax=<80 40 0>, angularLimitsMin=<-60 -40 0>, angularTarget=<0 0 0>}
      localJointOffset=<0 0 -10>
     }]
    sphericalLimits [{limitRadius=15, drivingBone {boneName=spine_05}, sphereLocalOffset=<2 -6 0>}]
   }
   n6 = Anim Dynamics - Bone: dyn_DVK_hair_05(
    ComponentPose=Reroute Node(InputPin:Component Space Pose Link Structure) -> (OutputPin:Component Space Pose Link Structure)
    Alpha=1
    ExternalForce="0,0,0"
   ) {
    angularDampingOverride=0.97
    angularSpringConstant=750
    bAngularSpring=true
    bGravityOverrideInSimSpace=true
    bOverrideAngularDamping=true
    bOverrideLinearDamping=true
    lODThreshold=1
    linearDampingOverride=0.97
    alphaBoolBlend {blendOption=Linear}
    alphaScaleBiasClamp {
     bClampResult=true
     bInterpResult=true
     clampMax=0.25
     clampMin=0.05
     interpSpeedDecreasing=0.5
     interpSpeedIncreasing=0.5
     outRange {max=0.1}
    }
    boundBone {boneName=dyn_DVK_hair_05}
    componentLinearAccScale=<1 1 1>
    componentLinearVelScale=<1 1 1>
    physicsBodyDefinitions [{
      sphereCollisionRadius=0
      boundBone {boneName=dyn_DVK_hair_05}
      boxExtents=<20 10 2>
      constraintSetup {
       angularTargetAxis=AxisY
       twistAxis=AxisY
       angularLimitsMax=<-80 12 0>
       angularLimitsMin=<-80 -12 0>
       angularTarget=<0 0 0>
       linearAxesMax=<10 10 10>
       linearAxesMin=<-10 -10 -10>
      }
      localJointOffset=<0 0 -10>
     }]
   }
   n7 = Anim Dynamics - Bone: dyn_DVK_hair_06(ComponentPose=@n6, Alpha=1, ExternalForce="0,0,0") {
    angularDampingOverride=0.96
    angularSpringConstant=750
    bAngularSpring=true
    bGravityOverrideInSimSpace=true
    bOverrideAngularDamping=true
    bOverrideLinearDamping=true
    lODThreshold=1
    linearDampingOverride=0.96
    alphaBoolBlend {blendOption=Linear}
    alphaScaleBiasClamp {
     bClampResult=true
     bInterpResult=true
     clampMax=0.25
     clampMin=0.05
     interpSpeedDecreasing=0.5
     interpSpeedIncreasing=0.5
     outRange {max=0.1}
    }
    boundBone {boneName=dyn_DVK_hair_06}
    componentLinearAccScale=<1 1 1>
    componentLinearVelScale=<1 1 1>
    physicsBodyDefinitions [{
      sphereCollisionRadius=0
      boundBone {boneName=dyn_DVK_hair_06}
      boxExtents=<20 10 2>
      constraintSetup {
       angularTargetAxis=AxisY
       twistAxis=AxisY
       angularLimitsMax=<-80 12 0>
       angularLimitsMin=<-80 -12 0>
       angularTarget=<0 0 0>
       linearAxesMax=<10 10 10>
       linearAxesMin=<-10 -10 -10>
      }
      localJointOffset=<0 0 -10>
     }]
   }
   n8 = Anim Dynamics - Bone: dyn_DVK_hair_07(ComponentPose=@n7, Alpha=1, ExternalForce="0,0,0") {
    angularDampingOverride=0.97
    angularSpringConstant=750
    bAngularSpring=true
    bGravityOverrideInSimSpace=true
    bOverrideAngularDamping=true
    bOverrideLinearDamping=true
    lODThreshold=1
    linearDampingOverride=0.97
    alphaBoolBlend {blendOption=Linear}
    alphaScaleBiasClamp {
     bClampResult=true
     bInterpResult=true
     clampMax=0.25
     clampMin=0.05
     interpSpeedDecreasing=0.5
     interpSpeedIncreasing=0.5
     outRange {max=0.1}
    }
    boundBone {boneName=dyn_DVK_hair_07}
    componentLinearAccScale=<1 1 1>
    componentLinearVelScale=<1 1 1>
    physicsBodyDefinitions [{
      sphereCollisionRadius=0
      boundBone {boneName=dyn_DVK_hair_07}
      boxExtents=<20 10 2>
      constraintSetup {
       angularTargetAxis=AxisY
       twistAxis=AxisY
       angularLimitsMax=<-80 12 0>
       angularLimitsMin=<-80 -12 0>
       angularTarget=<0 0 0>
       linearAxesMax=<10 10 10>
       linearAxesMin=<-10 -10 -10>
      }
      localJointOffset=<0 0 -10>
     }]
   }
   Anim Dynamics - Bone: dyn_DVK_hair_08(ComponentPose=@n8, Alpha=1, ExternalForce="0,0,0") {
    angularDampingOverride=0.96
    angularSpringConstant=750
    bAngularSpring=true
    bGravityOverrideInSimSpace=true
    bOverrideAngularDamping=true
    bOverrideLinearDamping=true
    lODThreshold=1
    linearDampingOverride=0.96
    alphaBoolBlend {blendOption=Linear}
    alphaScaleBiasClamp {
     bClampResult=true
     bInterpResult=true
     clampMax=0.25
     clampMin=0.05
     interpSpeedDecreasing=0.5
     interpSpeedIncreasing=0.5
     outRange {max=0.1}
    }
    boundBone {boneName=dyn_DVK_hair_08}
    componentLinearAccScale=<1 1 1>
    componentLinearVelScale=<1 1 1>
    physicsBodyDefinitions [{
      sphereCollisionRadius=0
      boundBone {boneName=dyn_DVK_hair_08}
      boxExtents=<20 10 2>
      constraintSetup {
       angularTargetAxis=AxisY
       twistAxis=AxisY
       angularLimitsMax=<-80 12 0>
       angularLimitsMin=<-80 -12 0>
       angularTarget=<0 0 0>
       linearAxesMax=<10 10 10>
       linearAxesMin=<-10 -10 -10>
      }
      localJointOffset=<0 0 -10>
     }]
   }
   n9 = Get PawnSpeed2D()
   n10 = Anim Dynamics - Bone: dyn_DVK_hair_12_a(Alpha=@n9, ExternalForce="0,0,0") {
    angularBiasOverride=1
    bGravityOverrideInSimSpace=true
    bOverrideAngularBias=true
    gravityScale=0.1
    alphaBoolBlend {blendOption=Linear}
    boundBone {boneName=dyn_DVK_hair_12_a}
    componentLinearAccScale=<1 1 1>
    componentLinearVelScale=<1 1 1>
    physicsBodyDefinitions [{
      sphereCollisionRadius=0
      boundBone {boneName=dyn_DVK_hair_12_a}
      boxExtents=<5 5 5>
      constraintSetup {
       angularTargetAxis=AxisY
       twistAxis=AxisY
       angularLimitsMax=<0.001 10 10>
       angularLimitsMin=<-0.001 -10 -10>
       angularTarget=<0 0 0>
       linearAxesMax=<0 0 3>
      }
      localJointOffset=<0 0 -0.5>
     }]
   }
   Anim Dynamics - Bone: dyn_DVK_hair_12(ComponentPose=@n10, Alpha=@n9, ExternalForce="0,0,0") {
    angularDampingOverride=0.96
    angularSpringConstant=300
    bGravityOverrideInSimSpace=true
    bOverrideAngularDamping=true
    bOverrideLinearDamping=true
    gravityScale=0.1
    lODThreshold=1
    linearDampingOverride=0.96
    linearSpringConstant=300
    alphaBoolBlend {blendOption=Linear}
    alphaScaleBiasClamp {
     bClampResult=true
     bInterpResult=true
     clampMax=0.25
     clampMin=0.05
     interpSpeedDecreasing=0.5
     interpSpeedIncreasing=0.5
     outRange {max=0.1}
    }
    boundBone {boneName=dyn_DVK_hair_12}
    componentLinearAccScale=<1 1 1>
    componentLinearVelScale=<1 1 1>
    physicsBodyDefinitions [{
      sphereCollisionRadius=0
      boundBone {boneName=dyn_DVK_hair_12}
      boxExtents=<30 30 30>
      constraintSetup {
       angularTargetAxis=AxisY
       twistAxis=AxisZ
       angularLimitsMax=<15 15 15>
       angularLimitsMin=<-15 -15 -15>
       angularTarget=<0 0 0>
       linearAxesMax=<10 10 10>
       linearAxesMin=<-10 -10 -10>
      }
      localJointOffset=<0 0 0>
     }]
   }
   n11 = Anim Dynamics - Bone: dyn_DVK_hair_09(Alpha=@n9, ExternalForce="0,0,0") {
    angularDampingOverride=0.98
    angularSpringConstant=325
    bAngularSpring=true
    bGravityOverrideInSimSpace=true
    bLinearSpring=true
    bOverrideAngularDamping=true
    bOverrideLinearDamping=true
    gravityScale=0
    lODThreshold=1
    linearDampingOverride=0.98
    linearSpringConstant=325
    alphaBoolBlend {blendOption=Linear}
    alphaScaleBiasClamp {
     bClampResult=true
     bInterpResult=true
     clampMax=0.25
     clampMin=0.05
     interpSpeedDecreasing=0.5
     interpSpeedIncreasing=0.5
     outRange {max=0.1}
    }
    boundBone {boneName=dyn_DVK_hair_09}
    componentLinearAccScale=<1 1 1>
    componentLinearVelScale=<1 1 1>
    physicsBodyDefinitions [{
      sphereCollisionRadius=0
      boundBone {boneName=dyn_DVK_hair_09}
      boxExtents=<30 30 30>
      constraintSetup {
       angularTargetAxis=AxisY
       twistAxis=AxisZ
       angularLimitsMax=<15 15 15>
       angularLimitsMin=<-15 -15 -15>
       angularTarget=<0 0 0>
       linearAxesMax=<10 10 10>
       linearAxesMin=<-10 -10 -10>
      }
      localJointOffset=<0 0 0>
     }]
   }
   n12 = Anim Dynamics - Bone: dyn_DVK_hair_10(ComponentPose=@n11, Alpha=@n9, ExternalForce="0,0,0") {
    angularDampingOverride=0.97
    angularSpringConstant=275
    bAngularSpring=true
    bGravityOverrideInSimSpace=true
    bLinearSpring=true
    bOverrideAngularDamping=true
    bOverrideLinearDamping=true
    gravityScale=0.15
    lODThreshold=1
    linearDampingOverride=0.97
    linearSpringConstant=275
    alphaBoolBlend {blendOption=Linear}
    alphaScaleBiasClamp {
     bClampResult=true
     bInterpResult=true
     clampMax=0.25
     clampMin=0.05
     interpSpeedDecreasing=0.5
     interpSpeedIncreasing=0.5
     outRange {max=0.1}
    }
    boundBone {boneName=dyn_DVK_hair_10}
    componentLinearAccScale=<1 1 1>
    componentLinearVelScale=<1 1 1>
    physicsBodyDefinitions [{
      sphereCollisionRadius=0
      boundBone {boneName=dyn_DVK_hair_10}
      boxExtents=<30 30 30>
      constraintSetup {
       angularTargetAxis=AxisY
       twistAxis=AxisZ
       angularLimitsMax=<15 15 15>
       angularLimitsMin=<-15 -15 -15>
       angularTarget=<0 0 0>
       linearAxesMax=<10 10 10>
       linearAxesMin=<-10 -10 -10>
      }
      localJointOffset=<0 0 0>
     }]
   }
   Anim Dynamics - Bone: dyn_DVK_hair_11(ComponentPose=@n12, Alpha=@n9, ExternalForce="0,0,0") {
    angularDampingOverride=0.96
    angularSpringConstant=300
    bAngularSpring=true
    bGravityOverrideInSimSpace=true
    bLinearSpring=true
    bOverrideAngularDamping=true
    bOverrideLinearDamping=true
    gravityScale=0.1
    lODThreshold=1
    linearDampingOverride=0.96
    linearSpringConstant=300
    alphaBoolBlend {blendOption=Linear}
    alphaScaleBiasClamp {
     bClampResult=true
     bInterpResult=true
     clampMax=0.25
     clampMin=0.05
     interpSpeedDecreasing=0.5
     interpSpeedIncreasing=0.5
     outRange {max=0.1}
    }
    boundBone {boneName=dyn_DVK_hair_11}
    componentLinearAccScale=<1 1 1>
    componentLinearVelScale=<1 1 1>
    physicsBodyDefinitions [{
      sphereCollisionRadius=0
      boundBone {boneName=dyn_DVK_hair_11}
      boxExtents=<30 30 30>
      constraintSetup {
       angularTargetAxis=AxisY
       twistAxis=AxisZ
       angularLimitsMax=<15 15 15>
       angularLimitsMin=<-15 -15 -15>
       angularTarget=<0 0 0>
       linearAxesMax=<10 10 10>
       linearAxesMin=<-10 -10 -10>
      }
      localJointOffset=<0 0 0>
     }]
   }
   n13 = Get gravity_override() -> (gravity_override:Vector)
   n14 = RigidBody!(ComponentPose=Local To Component(LocalPose=Use cached pose 'pre_sim_cache'()), Alpha, OverrideWorldGravity:Vector=@n13, ExternalForce) -> (Pose) {
    bClampLinearTranslationLimitToRefPose=true
    bOverrideWorldGravity=true
    overridePhysicsAsset=&/F_MED_Commando_DarkViking_Head_01_dyn_skydiving
    simulationSpace=BaseBoneSpace
    alphaBoolBlend {blendOption=Linear}
    baseBoneRef {boneName=head}
    componentAppliedLinearAccClamp=<3000 3000 3000>
    componentLinearAccScale=<0.35 0.2 0.2>
    componentLinearVelScale=<0.3 0.3 0.3>
    overrideWorldGravity {z=-980}
   }
   n15 = RigidBody(ComponentPose=Local To Component(LocalPose=Use cached pose 'pre_sim_cache'()), OverrideWorldGravity=@n13) {
    bClampLinearTranslationLimitToRefPose=true
    bOverrideWorldGravity=true
    overridePhysicsAsset=&/F_MED_Commando_DarkViking_Head_01_dyn_hair
    simulationSpace=BaseBoneSpace
    alphaBoolBlend {blendOption=Linear}
    baseBoneRef {boneName=head}
    componentAppliedLinearAccClamp=<3000 3000 3000>
    componentLinearAccScale=<0.35 0.2 0.2>
    componentLinearVelScale=<0.3 0.3 0.3>
    overrideWorldGravity {z=-980}
   }
   Output Pose(Result=Blend Poses by bool(BlendPose_0=Component To Local(ComponentPose=@n14), BlendPose_1=Component To Local(ComponentPose=@n15), bActiveValue=Get bIsSkyDiving()) {blendType=Linear}) {layerGroup=None}
   n16 = Anim Dynamics - Bone: dyn_DVK_hair_13(
    ComponentPose=Local To Component(LocalPose=Copy Pose From Mesh() {bCopyCurves=true, bUseAttachedParent=true})
    Alpha=0.25
    ExternalForce="0,0,0"
   ) {
    angularBiasOverride=0.98
    angularDampingOverride=0.98
    bGravityOverrideInSimSpace=true
    bOverrideAngularDamping=true
    bOverrideLinearDamping=true
    gravityScale=0.2
    linearDampingOverride=0.98
    alphaBoolBlend {blendOption=Linear}
    boundBone {boneName=dyn_DVK_hair_13}
    physicsBodyDefinitions [{
      sphereCollisionRadius=0
      boundBone {boneName=dyn_DVK_hair_13}
      boxExtents=<5 5 5>
      constraintSetup {angularTargetAxis=AxisY, twistAxis=AxisY, angularLimitsMax=<2 0 10>, angularLimitsMin=<-10 0 -10>, angularTarget=<0 0 0>}
      localJointOffset=<0 0 -2>
     }]
   }
   Save cached pose 'pre_sim_cache'(Pose=Component To Local(ComponentPose=@n16)) {cacheName=pre_sim_cache}
   n17 = Anim Dynamics - Bone: dyn_DVK_hair_01(Alpha=1, ExternalForce="0,0,0") {
    angularDampingOverride=0.97
    angularSpringConstant=750
    bAngularSpring=true
    bGravityOverrideInSimSpace=true
    bOverrideAngularDamping=true
    bOverrideLinearDamping=true
    lODThreshold=1
    linearDampingOverride=0.97
    alphaBoolBlend {blendOption=Linear}
    alphaScaleBiasClamp {
     bClampResult=true
     bInterpResult=true
     clampMax=0.25
     clampMin=0.05
     interpSpeedDecreasing=0.5
     interpSpeedIncreasing=0.5
     outRange {max=0.1}
    }
    boundBone {boneName=dyn_DVK_hair_01}
    componentLinearAccScale=<1 1 1>
    componentLinearVelScale=<1 1 1>
    physicsBodyDefinitions [{
      sphereCollisionRadius=0
      boundBone {boneName=dyn_DVK_hair_01}
      boxExtents=<20 10 2>
      constraintSetup {
       angularTargetAxis=AxisY
       twistAxis=AxisY
       angularLimitsMax=<-80 12 0>
       angularLimitsMin=<-80 -12 0>
       angularTarget=<0 0 0>
       linearAxesMax=<10 10 10>
       linearAxesMin=<-10 -10 -10>
      }
      localJointOffset=<0 0 -10>
     }]
   }
   n18 = Anim Dynamics - Bone: dyn_DVK_hair_02(ComponentPose=@n17, Alpha=1, ExternalForce="0,0,0") {
    angularDampingOverride=0.96
    angularSpringConstant=750
    bAngularSpring=true
    bGravityOverrideInSimSpace=true
    bOverrideAngularDamping=true
    bOverrideLinearDamping=true
    lODThreshold=1
    linearDampingOverride=0.96
    alphaBoolBlend {blendOption=Linear}
    alphaScaleBiasClamp {
     bClampResult=true
     bInterpResult=true
     clampMax=0.25
     clampMin=0.05
     interpSpeedDecreasing=0.5
     interpSpeedIncreasing=0.5
     outRange {max=0.1}
    }
    boundBone {boneName=dyn_DVK_hair_02}
    componentLinearAccScale=<1 1 1>
    componentLinearVelScale=<1 1 1>
    physicsBodyDefinitions [{
      sphereCollisionRadius=0
      boundBone {boneName=dyn_DVK_hair_02}
      boxExtents=<20 10 2>
      constraintSetup {
       angularTargetAxis=AxisY
       twistAxis=AxisY
       angularLimitsMax=<-80 12 0>
       angularLimitsMin=<-80 -12 0>
       angularTarget=<0 0 0>
       linearAxesMax=<10 10 10>
       linearAxesMin=<-10 -10 -10>
      }
      localJointOffset=<0 0 -10>
     }]
   }
   n19 = Anim Dynamics - Bone: dyn_DVK_hair_03(ComponentPose=@n18, Alpha=1, ExternalForce="0,0,0") {
    angularDampingOverride=0.97
    angularSpringConstant=750
    bAngularSpring=true
    bGravityOverrideInSimSpace=true
    bOverrideAngularDamping=true
    bOverrideLinearDamping=true
    lODThreshold=1
    linearDampingOverride=0.97
    alphaBoolBlend {blendOption=Linear}
    alphaScaleBiasClamp {
     bClampResult=true
     bInterpResult=true
     clampMax=0.25
     clampMin=0.05
     interpSpeedDecreasing=0.5
     interpSpeedIncreasing=0.5
     outRange {max=0.1}
    }
    boundBone {boneName=dyn_DVK_hair_03}
    componentLinearAccScale=<1 1 1>
    componentLinearVelScale=<1 1 1>
    physicsBodyDefinitions [{
      sphereCollisionRadius=0
      boundBone {boneName=dyn_DVK_hair_03}
      boxExtents=<20 10 2>
      constraintSetup {
       angularTargetAxis=AxisY
       twistAxis=AxisY
       angularLimitsMax=<-80 12 0>
       angularLimitsMin=<-80 -12 0>
       angularTarget=<0 0 0>
       linearAxesMax=<10 10 10>
       linearAxesMin=<-10 -10 -10>
      }
      localJointOffset=<0 0 -10>
     }]
   }
   n20 = Anim Dynamics - Bone: dyn_DVK_hair_04(ComponentPose=@n19, Alpha=1, ExternalForce="0,0,0") {
    angularDampingOverride=0.96
    angularSpringConstant=750
    bAngularSpring=true
    bGravityOverrideInSimSpace=true
    bOverrideAngularDamping=true
    bOverrideLinearDamping=true
    lODThreshold=1
    linearDampingOverride=0.96
    alphaBoolBlend {blendOption=Linear}
    alphaScaleBiasClamp {
     bClampResult=true
     bInterpResult=true
     clampMax=0.25
     clampMin=0.05
     interpSpeedDecreasing=0.5
     interpSpeedIncreasing=0.5
     outRange {max=0.1}
    }
    boundBone {boneName=dyn_DVK_hair_04}
    componentLinearAccScale=<1 1 1>
    componentLinearVelScale=<1 1 1>
    physicsBodyDefinitions [{
      sphereCollisionRadius=0
      boundBone {boneName=dyn_DVK_hair_04}
      boxExtents=<20 10 2>
      constraintSetup {
       angularTargetAxis=AxisY
       twistAxis=AxisY
       angularLimitsMax=<-80 12 0>
       angularLimitsMin=<-80 -12 0>
       angularTarget=<0 0 0>
       linearAxesMax=<10 10 10>
       linearAxesMin=<-10 -10 -10>
      }
      localJointOffset=<0 0 -10>
     }]
   }
   Reroute Node(InputPin=@n20)
  ]}
 EventGraph {nodes [
   Event BlueprintUpdateAnimation()
   > n1 = Gravity Override State Function DEPRECATED(CharacterPartAnimInstance=Self-Reference(), GameTimeInSeconds=GetGameTimeInSeconds())
   > Set gravity_override(gravity_override=@n1)
  ]}
]
variables [gravity_override {type=Vector}]
