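# Physics asset for the Dark Viking female medium commando body, a player character cosmetic.
# It holds only two bodies, one capsule on each lower-arm twist bone (left and right), with matching size and rotation and offsets in roughly opposite directions.
# It defines no constraints, so the forearm capsules are not jointed to each other and the rest of the skeleton has no bodies here.
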PhysicsAsset /BRCosmetics/Characters/Player/Female/Medium/Bodies/F_MED_Dark_Viking_01/Meshes/F_MED_Commando_DarkViking_Physics
bodies {
 lowerarm_twist_01_l [{shapeType=Capsule, length=16.9525, radius=4.7741, center=<-2.3802 0.4718 1.1048>, rotation=<rot 66.862 86.1981 86.5211>}]
 lowerarm_twist_01_r [{shapeType=Capsule, length=16.9525, radius=4.7741, center=<3.8637 -0.5151 -1.2047>, rotation=<rot 66.862 86.1981 86.5211>}]
}
constraints []
